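; Lake rune vent actor that launches players on foot and vehicles overlapping its capsule, negating their next falling damage.
; Players launch on authority along their own forward plus launch height, with skydiving ended and trail effects attached; vehicles use the vent's forward.
; Both paths fire a burst gameplay cue and optionally sphere-trace along the launch vector, applying shockwave grenade damage to clear obstacles.
; Tick, BeginPlay and replication/effect functions are empty.
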(event EventBeginPlay)

(event EventTick
  (params (DeltaSeconds :type Float%20%28single-precision%29)))

(event OnComponentBeginOverlap(Capsule)
  (params (OverlappedComponent :type Primitive%20Component%20Object%20Reference) (OtherActor :type Actor%20Object%20Reference) (OtherComp :type Primitive%20Component%20Object%20Reference) (OtherBodyIndex :type Integer) (bFromSweep :type Boolean) (SweepResult :type Hit%20Result%20Structure%20%28by%20ref%29))
  (bind _launchheight (Variables|Default|GetLaunchHeight))
  (bind _returnvalue (+ 0.0 _launchheight))
  (bind _launchxy_velocityscale (Variables|Default|GetLaunchXY_VelocityScale))
  (bind _playertarget (Variables|Default|GetPlayerTarget))
  (bind _returnvalue_2 (Transformation|GetActorForwardVector _playertarget))
  (bind _returnvalue_3 (* _returnvalue_2 _launchxy_velocityscale))
  (bind _y (.y _returnvalue_3))
  (bind _x (.x _returnvalue_3))
  (bind _returnvalue_4 (Math|Vector|MakeVector _x _y _returnvalue))
  (bind _vehiclexy_velocityscale (Variables|Default|GetVehicleXY_VelocityScale))
  (bind _returnvalue_5 (Transformation|GetActorForwardVector))
  (bind _returnvalue_6 (* _returnvalue_5 _vehiclexy_velocityscale))
  (bind _x_2 (.x _returnvalue_6))
  (bind _y_2 (.y _returnvalue_6))
  (bind _launchheight_2 (Variables|Default|GetLaunchHeight))
  (bind _returnvalue_7 (Math|Vector|MakeVector _x_2 _y_2 _launchheight_2))
  (bind _vehicleinterface (Variables|Default|GetVehicleInterface))
  (bind _playertarget_2 (Variables|Default|GetPlayerTarget))
  (bind _playertarget_3 (Variables|Default|GetPlayerTarget))
  (Components|Activation|Activate (Variables|Default|GetP_Athena_RuneVent_Trigger) false)
  (bind _asplayer_pawn_athena (Utilities|Casting|CastToPlayerPawn_Athena OtherActor)
    (:then
      (if (not (Vehicle|IsinVehicle _asplayer_pawn_athena))
        ;[1 1 1] Camera shake and Lens VFX.  Also needs SFX?
        (bind _output_get (Variables|Default|SetPlayerTarget _asplayer_pawn_athena))
        (bind _outcueparameters (GameplayCue|BuildCueParametersfromSource _output_get))
        (GameplayCue|ExecuteGameplayCue%28Burst%29Local (Variables|Default|GetPlayerTarget) (Variables|Default|GetRuneVentTag) _outcueparameters)
        ;[1 1 1] Play trail FX on player, destroy structures in path
        (Utilities|FlowControl|Sequence
          (:then_0
            (Components|Movement|StopMovementImmediately (Variables|/Script/Engine.Character|GetCharacterMovement _playertarget_2))
            (if (Variables|Default|GetShouldDestroyBuildings)
              (bind _outhits (Collision|MultiSphereTraceForObjects (Transformation|GetActorLocation _playertarget_2) (+ (Transformation|GetActorLocation _playertarget_2) (* (Math|Vector|Normalize _returnvalue_4) (Variables|Default|GetDestruction_Distance))) 1.5e+02 (Variables|Default|GetDestroyObjectTypes)))
              (for _array_element _outhits
                (if (Variables|/Script/Engine.Actor|GetbCanBeDamaged (.HitActor (Collision|BreakHitResult _array_element)))
                  (GameplayEffects|ApplyGameplayEffectToTarget :self (Ability|GetAbilitySystemComponent (Variables|Default|GetPlayerTarget)) :GameplayEffectClass "/Game/Athena/Items/Consumables/ShockwaveGrenade/GE_ShockGrenade_Damage.GE_ShockGrenade_Damage_C" :Target (Ability|GetAbilitySystemComponent (.HitActor (Collision|BreakHitResult _array_element))) :Level 0.0)))
              (else
                _))
            (Parachute|EndSkydiving _playertarget_3)
            (Variables|/Script/FortniteGame.FortPawn|SetbIgnoreNextFallingDamage true _playertarget_3)
            (Environment|ApplyGlobalEnvironmentGameplayEffecttoActor (Variables|Default|GetPlayerTarget) "/BlastBerryMapContent/Game/Athena/Items/Consumables/ShockwaveGrenade/GE_Athena_ShockGrenade_FX.GE_Athena_ShockGrenade_FX_C" 0 "(GameplayTags=)")
            (Networking|SwitchHasAuthority
              (:Authority
                (Character|LaunchCharacter (Variables|Default|GetPlayerTarget) _returnvalue_4 true true))))
          (:then_1
            (Effects|SpawnFXSystemAttached "/Game/Athena/Items/Consumables/KnockGrenade/NS_Athena_KnockGrenade_Player_Trails_Converted.NS_Athena_KnockGrenade_Player_Trails_Converted" (Variables|/Script/Engine.Character|GetMesh (Variables|Default|GetPlayerTarget)) "root")))))
    (:CastFailed
      ;[1 1 1] Vehicle Launch
      (bind _asfort_vehicle_interface (Utilities|Casting|CastToFortVehicleInterface OtherActor))
      (Variables|Default|SetVehicleInterface _asfort_vehicle_interface)
      (bind _outcueparameters_2 (GameplayCue|BuildCueParametersfromSource OtherActor))
      ;;[1 1 1] Camera shake and Lens VFX.  Also needs SFX?
      (GameplayCue|ExecuteGameplayCue%28Burst%29Local (Variables|Default|GetPlayerTarget) (Variables|Default|GetRuneVentTag) _outcueparameters_2)
      (if (Variables|Default|GetShouldDestroyBuildings)
        ;;[1 1 1] Trace up and destroy actors above
        (bind _outhits_2 (Collision|MultiSphereTraceForObjects (Class|FortVehicleInterface|GetOwnerLocation (Variables|Default|GetVehicleInterface)) (+ (Class|FortVehicleInterface|GetOwnerLocation (Variables|Default|GetVehicleInterface)) (* (Math|Vector|Normalize _returnvalue_7) (Variables|Default|GetDestruction_Distance))) 3.5e+02 (Variables|Default|GetDestroyObjectTypes)))
        (for _array_element_2 _outhits_2
          (if (Variables|/Script/Engine.Actor|GetbCanBeDamaged (.HitActor (Collision|BreakHitResult _array_element_2)))
            (GameplayEffects|ApplyGameplayEffectToTarget :self (Vehicle|GetVehicleAbilitySystemComponent (Variables|Default|GetVehicleInterface)) :GameplayEffectClass "/Game/Athena/Items/Consumables/ShockwaveGrenade/GE_ShockGrenade_Damage.GE_ShockGrenade_Damage_C" :Target (Ability|GetAbilitySystemComponent (.HitActor (Collision|BreakHitResult _array_element_2))) :Level 0.0))))
      ;;[1 1 1] Launch vehicle
      (Vehicle|SetIgnoreNextFallingDamage _vehicleinterface true)
      (Audio|BroadcastAbilityImpactatLocation 0 (Utilities|ToObject(Interface) _vehicleinterface) _returnvalue_7 7.8e+02))))

(fn OnRep_ForceActive)

(fn OnRep_VentIsActive)

(fn SetActiveEffects)

(fn ConstructionScript
  (Collision|SetCollisionProfileName (Variables|Default|GetCapsule) "OverlapAllDynamic" true :!class /Script/Engine.PrimitiveComponent))
